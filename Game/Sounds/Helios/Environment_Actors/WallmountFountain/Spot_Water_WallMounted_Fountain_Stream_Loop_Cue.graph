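# Looping ambient water spot sound cue for a wall-mounted fountain stream, routed to the water ambient sound class.
# Mixes a Coliseum fountain loop with a quieter close stream loop, then fades the blend in through an envelope and randomizes pitch slightly.
# All quality levels share the same chain.
# Inline overrides set focus-aware distance attenuation and cap concurrency at a few instances, stopping the lowest priority.
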SoundCue /Game/Sounds/Helios/Environment_Actors/WallmountFountain/Spot_Water_WallMounted_Fountain_Stream_Loop_Cue
root=QualityLevel_0
nodes [
 {
  id=Enveloper_1
  children [Mixer_0]
  properties {
   pitchCurve {editorCurveData {keys [{time=0}]}}
   volumeCurve {editorCurveData {keys [<at -0.0001 0 RCIM_Cubic>, {arriveTangent=0.4839, interpMode=RCIM_Cubic, leaveTangent=0.4839, tangentMode=RCTM_User, time=1, value=1}]}}
  }
 }
 {id=Mixer_0, children [WavePlayer_1,WavePlayer_2], properties {inputVolume [1,0.25]}}
 {id=Modulator_1, children [Enveloper_1], properties {pitchMax=1.1, pitchMin=0.9, volumeMax=1, volumeMin=1}}
 {id=QualityLevel_0, children [Modulator_1,Modulator_1,Modulator_1,Modulator_1]}
 {id=WavePlayer_1, properties {bLooping=true, soundWaveAssetPtr=/Game/Sounds/Apollo/Biomes/Coliseum/Coliseum_Fountain_Loop_01}}
 {id=WavePlayer_2, properties {bLooping=true, soundWaveAssetPtr=/Game/Athena/Sounds/Water/Ambient/Amb_Water_Stream_Med_Close_Loop_02}}
]
properties {
 bOverrideAttenuation=true
 bOverrideConcurrency=true
 maxDistance=1750
 priority=10
 soundClassObject=/Game/Sounds/FortSoundClassesAndModes/Fort_Water_Ambient
 volumeMultiplier=1
 attenuationOverrides {bEnableListenerFocus=true, falloffDistance=1500, nonFocusVolumeAttenuation=0.75, attenuationShapeExtents {x=250}}
 concurrencyOverrides {maxCount=2, resolutionRule=StopLowestPriority}
}
